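# Masked lit material for mesh particles in the Bubblegum pickaxe effects, tinted by particle color.
# A time-of-day function darkens that color for evening and night, and a Fresnel blend with a scaled copy feeds emissive adjustments, zeroed at low quality.
# A dynamic Erosion parameter, inverted, scales smoothstepped cellular and inverted electric noise, whose saturated sum times particle alpha forms a dissolve mask.
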Material /BRCosmetics/Weapons/FORT_Melee/Pickaxe_Bubblegum/FX/M_BubbleGum
graph {
 properties {bUsedWithMeshParticles=true, blendMode=BLEND_Masked, pixelDepthOffsetMode=PDOM_Legacy, usedShadingModels=Default Lit}
 nodes [
  n1 = ParticleColor()
  n2 = Multiply(A=@n1.RGB, B=ScalarParameter() {desc=Multiplies BaseColor by a float to darken for Evening/Night, parameterName=Evening/Night Color Mult}) {constB=0.7}
  n3 = MaterialFunctionCall(Color_Morning=@n1.RGB, Color_Day=@n1.RGB, Color_Evening=@n2, Color_Night=@n2) {materialFunction=&Color_Time_of_Day}
  n4 = Multiply(A=@n3.Result, B=ScalarParameter() {defaultValue=-0.25, desc=Makes Fresnel glow w/ BaseColor, parameterName=Emissive})
  n5 = Fresnel(ExponentIn=ScalarParameter() {defaultValue=1, parameterName=Fresnel Exponent}) {baseReflectFraction=0}
  MaterialFunctionCall(Emissive=QualitySwitch(Default=LinearInterpolate(A=@n3.Result, B=@n4, Alpha=@n5), Low=Constant())) {materialFunction=&MF_FX_EmissiveAdjustments}
  n6 = OneMinus(None=DynamicParameter() {paramNames [Erosion,not_used,not_used,not_used]}.Erosion)
  n7 = Multiply(
   A=@n6
   B=MaterialFunctionCall!(In=TextureSample() {samplerType=SAMPLERTYPE_Masks, texture=/Game/Global/Textures/Default/Noise/T_CellularNoise}.G) {materialFunction=/Game/Effects/Fort_Effects/Materials/_Functions/MF_Smoothstep_01}.Result
  )
  n8 = TextureSample() {samplerType=SAMPLERTYPE_LinearColor, texture=/Game/Characters/Player/Male/Medium/Bodies/M_MED_SRDrift/Textures/T_ElectricNoise}
  n9 = Multiply(
   A=MaterialFunctionCall(In=OneMinus(None=@n8.G)) {materialFunction=&MF_Smoothstep_01}.Result
   B=@n6
  )
  Multiply(A=@n1.A, B=Saturate(None=Add(A=@n7, B=@n9)))
  ScalarParameter() {parameterName=Metallic}
  ScalarParameter() {defaultValue=1, parameterName=Roughness}
 ]
}
