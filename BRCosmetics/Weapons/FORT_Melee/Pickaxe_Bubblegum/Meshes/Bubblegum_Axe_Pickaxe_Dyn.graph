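# Physics asset for the Bubblegum pickaxe cosmetic, defining bodies for five gumballs and the container around them.
# Each dyn_gumball bone has a small sphere body, while mstr_gumball approximates a hollow globe with a floor box, a top box and eight sides of three angled boxes.
# Each gumball is constrained to mstr_gumball with free swing and twist, building the concave shell from convex primitives.
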PhysicsAsset /BRCosmetics/Weapons/FORT_Melee/Pickaxe_Bubblegum/Meshes/Bubblegum_Axe_Pickaxe_Dyn
bodies {
 dyn_gumball_1 [{shapeType=Sphere, radius=2.4}]
 dyn_gumball_2 [{shapeType=Sphere, radius=2.4}]
 dyn_gumball_3 [{shapeType=Sphere, radius=2.4}]
 dyn_gumball_4 [{shapeType=Sphere, radius=2.4}]
 dyn_gumball_5 [{shapeType=Sphere, radius=2.4}]
 mstr_gumball [
  {shapeType=Box, extentX=12.625, extentY=13.4375, extentZ=2.5, center=<0 0 -1.1835>}
  {shapeType=Box, extentX=6.625, extentY=14.0625, extentZ=2.5, center=<7.8684 -0.0001 1.2516>, rotation=<rot 39.9998 0 0>}
  {shapeType=Box, extentX=6.625, extentY=14.0625, extentZ=2.5, center=<10.2773 -0.0001 4.8028>, rotation=<rot 74.9999 0 0>}
  {shapeType=Box, extentX=18, extentY=19.375, extentZ=3.4375, center=<0.0397 0 14.3209>}
  {shapeType=Box, extentX=6.625, extentY=14.0625, extentZ=2.5, center=<9.7238 -0.0001 10.2361>, rotation=<rot 59.9903 179.9999 180>}
  {shapeType=Box, extentX=6.625, extentY=14.0625, extentZ=2.5, center=<-8.0264 -0.0001 1.2516>, rotation=<rot 39.9987 -179.9995 0>}
  {shapeType=Box, extentX=6.625, extentY=14.0625, extentZ=2.5, center=<-10.1267 -0.0001 4.8028>, rotation=<rot 74.9987 179.999 0.0001>}
  {shapeType=Box, extentX=6.625, extentY=14.0625, extentZ=2.5, center=<-9.3919 -0.0001 10.2361>, rotation=<rot 59.9884 -0.0002 179.9999>}
  {shapeType=Box, extentX=6.625, extentY=14.0625, extentZ=2.5, center=<-1.6413 7.886 1.2516>, rotation=<rot 39.9997 89.9994 0>}
  {shapeType=Box, extentX=6.625, extentY=14.0625, extentZ=2.5, center=<0.7675 9.8818 4.8028>, rotation=<rot 74.9989 89.9994 0.0001>}
  {shapeType=Box, extentX=6.625, extentY=14.0625, extentZ=2.5, center=<0.214 9.4628 10.2361>, rotation=<rot 59.9854 -89.9969 180>}
  {shapeType=Box, extentX=6.625, extentY=14.0625, extentZ=2.5, center=<0.1262 -7.9515 1.2516>, rotation=<rot 39.9995 -90.0011 0.0001>}
  {shapeType=Box, extentX=6.625, extentY=14.0625, extentZ=2.5, center=<-0.2854 -10.255 4.8028>, rotation=<rot 74.9984 -90.001 0.0004>}
  {shapeType=Box, extentX=6.625, extentY=14.0625, extentZ=2.5, center=<-0.6678 -9.4496 10.2361>, rotation=<rot 59.9849 90.0035 179.9999>}
  {shapeType=Box, extentX=6.625, extentY=14.0625, extentZ=2.5, center=<6.2718 -5.4113 1.2518>, rotation=<rot 39.9997 -45 0>}
  {shapeType=Box, extentX=6.625, extentY=14.0625, extentZ=2.5, center=<7.4514 -6.6404 4.8029>, rotation=<rot 74.9987 -45.0003 0>}
  {shapeType=Box, extentX=6.625, extentY=14.0625, extentZ=2.5, center=<7.1325 -6.4058 10.2363>, rotation=<rot 59.9822 134.9982 180>}
  {shapeType=Box, extentX=6.625, extentY=14.0625, extentZ=2.5, center=<-5.7845 6.1728 1.2518>, rotation=<rot 39.9974 134.9998 0>}
  {shapeType=Box, extentX=6.625, extentY=14.0625, extentZ=2.5, center=<-6.8684 7.1892 4.803>, rotation=<rot 74.9957 134.9979 0>}
  {shapeType=Box, extentX=6.625, extentY=14.0625, extentZ=2.5, center=<-6.5509 6.7719 10.2363>, rotation=<rot 59.9879 -45.0009 179.9999>}
  {shapeType=Box, extentX=6.625, extentY=14.0625, extentZ=2.5, center=<4.9775 4.9616 1.2518>, rotation=<rot 39.9992 44.9987 0>}
  {shapeType=Box, extentX=6.625, extentY=14.0625, extentZ=2.5, center=<7.1541 7.1896 4.803>, rotation=<rot 74.9981 44.9992 0.0002>}
  {shapeType=Box, extentX=6.625, extentY=14.0625, extentZ=2.5, center=<6.6007 6.7705 10.2363>, rotation=<rot 59.9778 -134.991 180>}
  {shapeType=Box, extentX=6.625, extentY=14.0625, extentZ=2.5, center=<-5.9024 -5.5633 1.2517>, rotation=<rot 39.9984 -135.0003 0.0001>}
  {shapeType=Box, extentX=6.625, extentY=14.0625, extentZ=2.5, center=<-7.1832 -6.9976 4.803>, rotation=<rot 74.9963 -135 0.0005>}
  {shapeType=Box, extentX=6.625, extentY=14.0625, extentZ=2.5, center=<-6.9086 -6.8491 10.2363>, rotation=<rot 59.9848 45.0043 179.9999>}
 ]
}
constraints [
 {
  bone1Name=dyn_gumball_5
  bone2Name=mstr_gumball
  swing1LimitDegrees=45
  swing1Motion=Free
  swing2LimitDegrees=45
  swing2Motion=Free
  twistLimitDegrees=45
  twistMotion=Free
 }
 {
  bone1Name=dyn_gumball_4
  bone2Name=mstr_gumball
  swing1LimitDegrees=45
  swing1Motion=Free
  swing2LimitDegrees=45
  swing2Motion=Free
  twistLimitDegrees=45
  twistMotion=Free
 }
 {
  bone1Name=dyn_gumball_3
  bone2Name=mstr_gumball
  swing1LimitDegrees=45
  swing1Motion=Free
  swing2LimitDegrees=45
  swing2Motion=Free
  twistLimitDegrees=45
  twistMotion=Free
 }
 {
  bone1Name=dyn_gumball_2
  bone2Name=mstr_gumball
  swing1LimitDegrees=45
  swing1Motion=Free
  swing2LimitDegrees=45
  swing2Motion=Free
  twistLimitDegrees=45
  twistMotion=Free
 }
 {
  bone1Name=dyn_gumball_1
  bone2Name=mstr_gumball
  swing1LimitDegrees=45
  swing1Motion=Free
  swing2LimitDegrees=45
  swing2Motion=Free
  twistLimitDegrees=45
  twistMotion=Free
 }
]
