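# Animation Blueprint for the Bubblegum pickaxe that animates its plunger and bubble bones in response to hits.
# Each update reads SuccessfulHit and HitCounter from the owning pickaxe actor, pulsing PlungerAlpha on a hit.
# A Do Once, rearmed when SuccessfulHit is false, grows BalloonScale while HitCounter is positive, otherwise shrinks it.
# The AnimGraph additively scales dyn_bubble, offsets dyn_front with interpolated alpha, then applies RigidBody physics.
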AnimBlueprint /BRCosmetics/Weapons/FORT_Melee/Pickaxe_Bubblegum/Meshes/Bubblegum_Axe_Pickaxe_AnimBP
target_skeleton=&/Bubblegum_Axe_Pickaxe_Skeleton
graphs [
 AnimGraph {nodes [
   AnimGraphNode_SubInput_0 = Input Pose()
   n1 = "Transform (Modify) Bone - Bone: dyn_bubble"(
    ComponentPose=Local To Component(LocalPose=@AnimGraphNode_SubInput_0)
    Alpha=1
    Translation="0,0,0"
    Rotation="0,0,0"
    Scale=Get BalloonScale() -> (BalloonScale:Vector)
   ) {scaleMode=BMM_Additive, scaleSpace=BCS_BoneSpace, alphaBoolBlend {blendOption=Linear}, boneToModify {boneName=dyn_bubble}}
   n2 = "Transform (Modify) Bone - Bone: dyn_front"(ComponentPose=@n1, Alpha=Get PlungerAlpha() -> (PlungerAlpha:double), Translation="0,-12,0", Rotation="0,0,0", Scale="1,1,1") {
    translationMode=BMM_Additive
    translationSpace=BCS_BoneSpace
    alphaBoolBlend {blendOption=Linear}
    alphaScaleBiasClamp {bInterpResult=true, interpSpeedDecreasing=2, interpSpeedIncreasing=15}
    boneToModify {boneName=dyn_front}
    translation {y=-12}
   }
   n3 = RigidBody(ComponentPose=@n2) {
    bOverrideWorldGravity=true
    overridePhysicsAsset=&/Bubblegum_Axe_Pickaxe_Dyn
    alphaBoolBlend {blendOption=Linear}
    componentAppliedLinearAccClamp=<3000 3000 3000>
    componentLinearAccScale=<0.5 0.5 0.5>
    componentLinearVelScale=<0.5 0.5 0.5>
    overrideWorldGravity {z=-980}
   }
   Output Pose(Result=Component To Local(ComponentPose=@n3)) {layerGroup=None}
  ]}
 EventGraph {nodes [
   Event BlueprintUpdateAnimation()
   > n1 = Cast To B_Melee_Impact_Pickaxe_Athena_Bubblegum!(execute, Object=GetOwningActor()) -> (then, CastFailed, AsB Melee Impact Pickaxe Athena Bubblegum:B Melee Impact Pickaxe Athena Bubblegum Object Reference) {is=title K2Node_DynamicCast}
   > n2 = Branch(Condition=Get SuccessfulHit!(self:B Melee Impact Pickaxe Athena Bubblegum Object Reference=@n1) -> (SuccessfulHit:bool))
   > n3 = Sequence()
   > Set PlungerAlpha(PlungerAlpha:double=1) -> (Output_Get:double)
   > Delay(Duration=0.2)
   > Set PlungerAlpha(PlungerAlpha=0)
   Do Once(execute=@n3.then_1, Reset=@n2.else)
   > n4 = Branch(Condition="integer > integer"(A=Get HitCounter!(self:B Melee Impact Pickaxe Athena Bubblegum Object Reference=@n1) -> (HitCounter:int=0), B=0) {is=title K2Node_CallFunction})
   > Set BalloonScale(BalloonScale:Vector=vector + vector(A:Vector=Get BalloonScale(), B:Vector="0.3,0.3,0.3") -> (ReturnValue:Vector) {is=title K2Node_CommutativeAssociativeBinaryOperator}) -> (Output_Get:Vector)
   Set BalloonScale(execute=@n4.else, BalloonScale="0.1,0.1,0.1")
   Event BlueprintInitializeAnimation()
   > n5 = Cast To B_Melee_Impact_Pickaxe_Athena_Bubblegum(Object=GetOwningActor())
   TryGetPawnOwner()
   Get SuccessfulHit(self=@n5)
  ]}
]
variables [
 BalloonScale {type=Vector, default=<1 1 1>}
 PlungerAlpha {type=number}
]
